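# Niagara system for the DecoBlue vehicle boost cosmetic, emitting trails, sparkles and a flare from the car's boost sockets.
# Gold-blue and wind ribbon emitters stream backward, with game-mode-dependent tuning and a travelling vertical wave whose amplitude follows camera alignment.
# Two sprite emitters scatter curl-noise particles partitioned across ports, while LensFlare holds burst-spawned sprites at the sockets.
# Painted color overrides are active only on the first sprite emitter.
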NiagaraSystem /VehicleCosmetics/Boosts/Boost_DecoBlue/NS_DecoBlue
dependencies {
 usedDataInterfaces [
  &NiagaraDataInterfaceCurve
  &NiagaraDataInterfaceColorCurve
  &NiagaraDataInterfaceArrayFloat
  &NiagaraDataInterfaceCamera
  &NiagaraDataInterfaceSkeletalMesh
 ]
 usedDynamicInputs []
 usedModules [
  &self:SystemSpawnScript.SystemScriptSource.SystemScriptGraph.NiagaraNodeAssignment_0.SetVariables_#1
  &SystemState
 ]
}
emitters {
 LensFlare {
  emitterUpdateScript [
   &EmitterState(
    Inactive Response=::NewEnumerator1
    Life Cycle Mode=::NewEnumerator1
    Loop Behavior=::NewEnumerator1
    Loop Duration Mode=::NewEnumerator1
    UseLoopDelay=false
   )
   &VC_FilteredSocketCount()
   &SpawnBurst_Instantaneous(Spawn Count=2, Use Spawn Probability=false, UseLoopCountLimit=false)
   &SpawnRate(SpawnRate=300, Use Spawn Probability=false) {enabled=false}
  ]
  particleSpawnScript [
   &VC_PaintedVariantsOverride(DefaultColor=<rgba 0.2213 0.5145 0.6522 0.2>) {enabled=false}
   &Color(Color=<rgba 0.2213 0.5145 0.6522 0.1>)
   &InitializeParticle(
    Color Mode=::NewEnumerator0
    Position Mode=::NewEnumerator0
    Sprite Rotation Mode=::NewEnumerator0
    Sprite Size=<130 90>
    Sprite Size Mode=::NewEnumerator4
    Write Material Random=true
   )
   &PartitionParticles(Number Of Partitions=$Emitter.PortCount)
   &SkeletalMeshLocation(
    Bone / Socket Index=$Particles.PartitionParticles.Partition
    Bone / Socket Sampling Mode=::NewEnumerator5
    MaskSpawnGroup=false
    Mesh Sampling Type=::NewEnumerator6
    Sampled Position Offset=&AddVectorToPosition(Position=<-50 0 0>, Vector=Linked() {linkedVariable {name=User.DTCarOffset001, type=NiagaraPosition}})
    Skeletal Mesh=$System.SkeletalMesh
   )
   &AddVelocity(Velocity=<-1500 0 0>) {enabled=false}
  ]
  particleUpdateScript [
   &ParticleState(Kill Particles When Lifetime Has Elapsed=false, Loop Particles Lifetime=true)
   &ScaleSpriteSize(Uniform Curve Sprite Scale=DataInterface() {propertyValues {Curve {keys [{tangentMode=RCTM_User, time=0, value=2}, <at 0.0927 1>, {tangentMode=RCTM_User, time=1, value=1}]}}}) {enabled=false}
   &SolveForcesAndVelocity(Clamp Velocity=false, Limit Acceleration=false)
   &ScaleColor(
    Scale RGB=&VectorFromFloat(Value=&FloatFromCurve(FloatCurve=DataInterface() {propertyValues {Curve {keys [<at 0 0 RCIM_Cubic>,<at 0.3 1 RCIM_Cubic>,<at 1 0 RCIM_Cubic>]}}}))
    ScaleA=true
    ScaleRGB=true
   ) {enabled=false}
  ]
  rendererClasses [&NiagaraSpriteRendererProperties]
  renderers [{rendererClass=&NiagaraSpriteRendererProperties}]
 }
 NE_VC_Boost_Ribbon_GoldBlue {
  emitterUpdateScript [
   &EmitterState(Life Cycle Mode=::NewEnumerator1, Loop Duration=1, UseLoopDelay=false)
   &VC_FilteredSocketCount()
   &VC_RateQualityHandler(High Spawn Rate=60, Low Spawn Rate=60, Medium Spawn Rate=60, Use Custom Rate?=true)
   &VC_SpawnRibbon_v1(Number of Ribbons=$Emitter.PortCount, SpawnRate=$Emitter.Spawn Rate)
  ]
  particleSpawnScript [
   &VC_PaintedVariantsOverride(
    DefaultColor=&ColorFromCurve(ColorCurve=DataInterface() {propertyValues {
       AlphaCurve {keys [<at 0 0>,<at 0.0745 1 RCIM_Cubic>,<at 0.8087 0 RCIM_Cubic>]}
       BlueCurve {keys [<at 0 1.6 RCIM_Cubic>,<at 0.3 0.6 RCIM_Cubic>]}
       GreenCurve {keys [<at 0 4 RCIM_Cubic>,<at 0.3 2 RCIM_Cubic>]}
       RedCurve {keys [<at 0 2 RCIM_Cubic>,<at 0.3 0 RCIM_Cubic>]}
      }})
    R1 DefaultColor=<rgba 0 1 0.5972 0.7765>
    R2 DefaultColor:LinearColor=<rgba 1 0 0.8 1>
    Use R1 Painted Overrides=true
    Use R2 Painted Overrides=true
   ) {enabled=false}
   &VC_InitializeRibbon_v1(
    Color=<rgba 0.9 0.6195 0.135 1>
    Custom Offset=true
    Enable Skeletal Position=true
    Facing Axis:ENiagaraOrientationAxis=::YAxis
    Lifetime=&MergeHelper_FloatSwitch(0=0.85, 1=0.8, 2=0.8, 3=?, Index=$System.GameModeID, Num=::NewEnumerator1)
    Local Vector Offset=&AddVectorToPosition(
     Position=&MergeHelper_VectorSwitch(0=<-12 0 0>, 1=<-40 0 0>, 2=<-40 0 0>, 3=?, Index=$System.GameModeID, Num=::NewEnumerator1)
     Vector=Linked() {linkedVariable {name=User.DTCarOffset002, type=NiagaraPosition}}
    )
    Offset Amount=1
    Ribbon Width=90
   )
   &AddVelocity(Velocity=&MergeHelper_VectorSwitch(0=<-1000 0 0>, 1=<-1000 0 0>, 2=<-1000 0 0>, 3=?, Index=$System.GameModeID, Num=::NewEnumerator1)) {module=AddVelocity001}
   /SetVariables(
    Particles.R1_Width:NiagaraFloat=&SelectFloatFromArray(Direct Array Index=&SpawnGroup(), Float Selection Array=DataInterface() {propertyValues {FloatData [1,0]}})
    Particles.R2_Width:NiagaraFloat=&SelectFloatFromArray(Direct Array Index=&SpawnGroup(), Float Selection Array=DataInterface() {propertyValues {FloatData [0,-1]}})
    Particles.R3_Color:LinearColor=<rgba 0.2241 0.5211 0.6609 3>
   ) {isSetParameters=true}
  ]
  particleUpdateScript [
   &ParticleState()
   &InheritSourceMovement(
    Applied Position Scale=&MergeHelper_VectorSwitch(
     0=&Multiply_VectorByFloat(
      Vector=<1 1 1>
      Float=&FloatFromCurve(FloatCurve=DataInterface() {propertyValues {Curve {keys [{interpMode=RCIM_Cubic, tangentMode=RCTM_Break, time=0.5, value=1}, {interpMode=RCIM_Cubic, tangentMode=RCTM_Break, time=1, value=0.5}]}}})
     )
     1=&Multiply_VectorByFloat(
      Vector=<1 1 1>
      Float=&FloatFromCurve(FloatCurve=DataInterface() {propertyValues {Curve {keys [{interpMode=RCIM_Cubic, tangentMode=RCTM_Break, time=0.5, value=1}, {interpMode=RCIM_Cubic, tangentMode=RCTM_Break, time=1, value=0.5}]}}})
     )
     2=<0 0 0>
     3=?
     Index=0
     Num=::NewEnumerator1
    )
    ApplyForce=false
    ApplyPosition=true
    ApplyPositionFalloff=false
   )
   &SolveForcesAndVelocity(Clamp Velocity=false, Limit Acceleration=false)
   /SetVariables(Particles.DotProduct:NiagaraFloat=&DotProduct(
     Vector A=/Niagara/DynamicInputs/Math/TransformVectorByMatrix(
      Vector:Vector3f=<1 0 0>
      Transformation Matrix:NiagaraMatrix=$Engine.Owner.SystemLocalToWorld
      Transform Vector as Position:NiagaraBool=false
     )
     Vector B=&NormalizeVector(
      Vector To Normalize=&SubtractVectorFromPosition(
       Position=$Engine.Owner.Position
       Vector=&ReturnCameraPosition(Coordinate Space:ENiagaraCoordinateSpace=::Simulation, Camera Query:NiagaraDataInterfaceCamera=DataInterface())
      )
      Safe Normalize=true
      Safe Fallback Threshold=0
      Safe Fallback Vector=<0 0 0.001>
     )
    )) {isSetParameters=true}
   /SetVariables(Particles.CustomPosition_R1:NiagaraPosition=&AddVectorToPosition(
     Position=$Particles.Position
     Vector=&MakeVector(
      X=0
      Y=0
      Z=&MaskFloat(
       Mask Source=0
       Mask Parameter=?
       Mask=0
       Masked Float=&Cosine(
        Normalized Angle=&Add_Float(A=0, B=&Subtract_Float(A=$Emitter.NormalizedLoopAge, B=$Particles.NormalizedAge))
        Period=1
        Scale=&FloatFromCurve(
         FloatCurve=DataInterface() {propertyValues {Curve {keys [
             {arriveTangent=4.1101, interpMode=RCIM_Cubic, leaveTangent=4.1101, tangentMode=RCTM_User, time=0.0999, value=0.71}
             {arriveTangent=-0.294, interpMode=RCIM_Cubic, leaveTangent=-0.294, time=0.5, value=1.6}
             {
              arriveTangent=-0.5881
              arriveTangentWeight=0.5973
              interpMode=RCIM_Cubic
              leaveTangent=-0.5881
              leaveTangentWeight=0.5973
              tangentMode=RCTM_User
              time=0.7585
              value=0.5164
             }
            ]}}}
         Scale Curve=&Lerp_Float(
          A=5
          B=6.25
          Alpha=&NormalizeFloat(Value=&Power(a=&Add_Float(A=$Particles.DotProduct, B=0.1), b=5), Min=0, Max=1)
         )
        )
        Bias=0
       )
       Passthrough Float=&Sine(
        Normalized Angle=&Add_Float(A=0, B=&Subtract_Float(A=$Emitter.NormalizedLoopAge, B=$Particles.NormalizedAge))
        Period=1
        Scale=&FloatFromCurve(
         FloatCurve=DataInterface() {propertyValues {Curve {keys [
             {arriveTangent=4.1101, interpMode=RCIM_Cubic, leaveTangent=4.1101, tangentMode=RCTM_User, time=0.0999, value=0.71}
             {arriveTangent=-0.294, interpMode=RCIM_Cubic, leaveTangent=-0.294, time=0.5, value=1.6}
             {
              arriveTangent=-0.5881
              arriveTangentWeight=0.5973
              interpMode=RCIM_Cubic
              leaveTangent=-0.5881
              leaveTangentWeight=0.5973
              tangentMode=RCTM_User
              time=0.7585
              value=0.5164
             }
            ]}}}
         Scale Curve=&Lerp_Float(
          A=5
          B=6.25
          Alpha=&NormalizeFloat(Value=&Power(a=&Add_Float(A=$Particles.DotProduct, B=0.1), b=5), Min=0, Max=1)
         )
        )
        Bias=0
       )
      )
     )
    )) {isSetParameters=true}
   &VC_UpdateRibbon_v1(Enable Skeletal Position=true)
   /SetVariables(
    Particles.R1_Width:NiagaraFloat=&Multiply_Float(
     A=$Particles.Initial.R1_Width
     B=&FloatFromCurve(
      FloatCurve=DataInterface() {propertyValues {Curve {keys [
          {interpMode=RCIM_Cubic, leaveTangentWeight=0.07, tangentMode=RCTM_Break, time=0, value=0.3}
          {
           arriveTangent=0.1655
           arriveTangentWeight=0.092
           interpMode=RCIM_Cubic
           leaveTangent=0.1655
           leaveTangentWeight=0.1363
           tangentMode=RCTM_User
           time=0.14
           value=1.3
          }
          {
           arriveTangent=0.1594
           arriveTangentWeight=0.1489
           interpMode=RCIM_Cubic
           leaveTangent=0.1594
           leaveTangentWeight=0.2228
           tangentMode=RCTM_User
           time=0.35
           value=1.4
          }
          {arriveTangent=-0.1198, interpMode=RCIM_Cubic, leaveTangent=-0.1198, tangentMode=RCTM_User, time=0.7, value=1.75}
          {arriveTangentWeight=0.1553, interpMode=RCIM_Cubic, time=1, value=0.25}
         ]}}}
      Scale Curve=55
     )
    )
    Particles.R2_Width:NiagaraFloat=&Multiply_Float(
     A=$Particles.Initial.R2_Width
     B=&FloatFromCurve(
      FloatCurve=DataInterface() {propertyValues {Curve {keys [
          {interpMode=RCIM_Cubic, leaveTangentWeight=0.07, tangentMode=RCTM_Break, time=0, value=0.3}
          {
           arriveTangent=0.1655
           arriveTangentWeight=0.092
           interpMode=RCIM_Cubic
           leaveTangent=0.1655
           leaveTangentWeight=0.1363
           tangentMode=RCTM_User
           time=0.14
           value=1.3
          }
          {
           arriveTangent=0.1594
           arriveTangentWeight=0.1489
           interpMode=RCIM_Cubic
           leaveTangent=0.1594
           leaveTangentWeight=0.2228
           tangentMode=RCTM_User
           time=0.35
           value=1.4
          }
          {arriveTangent=-0.1198, interpMode=RCIM_Cubic, leaveTangent=-0.1198, tangentMode=RCTM_User, time=0.7, value=1.75}
          {arriveTangentWeight=0.1553, interpMode=RCIM_Cubic, time=1, value=0.25}
         ]}}}
      Scale Curve=55
     )
    )
    Particles.R3_Width:NiagaraFloat=&Multiply_Float(
     A=85
     B=&FloatFromCurve(FloatCurve=DataInterface() {propertyValues {Curve {keys [<at 0 0.5>,<at 0.25 1>]}}})
    )
   ) {isSetParameters=true, module=SetVariables_#2}
  ]
  rendererClasses [&NiagaraRibbonRendererProperties]
  renderers [
   {rendererClass=&NiagaraRibbonRendererProperties}
   {rendererClass=&NiagaraRibbonRendererProperties, rendererIndex=1}
   {rendererClass=&NiagaraRibbonRendererProperties, rendererIndex=2}
  ]
 }
 NE_VC_Boost_Ribbon_Wind {
  emitterUpdateScript [
   &EmitterState(Life Cycle Mode=::NewEnumerator1, Loop Duration=1, UseLoopDelay=false)
   &VC_FilteredSocketCount()
   &VC_RateQualityHandler(Epic Spawn Rate=30, High Spawn Rate=30, Low Spawn Rate=30, Use Custom Rate?=true)
   &VC_SpawnRibbon_v1(Number of Ribbons=$Emitter.PortCount, SpawnRate=$Emitter.Spawn Rate)
  ]
  particleSpawnScript [
   &VC_PaintedVariantsOverride(
    DefaultColor=&ColorFromCurve(ColorCurve=DataInterface() {propertyValues {
       AlphaCurve {keys [<at 0 0>,<at 0.0745 1 RCIM_Cubic>,<at 0.8087 0 RCIM_Cubic>]}
       BlueCurve {keys [<at 0 1.6 RCIM_Cubic>,<at 0.3 0.6 RCIM_Cubic>]}
       GreenCurve {keys [<at 0 4 RCIM_Cubic>,<at 0.3 2 RCIM_Cubic>]}
       RedCurve {keys [<at 0 2 RCIM_Cubic>,<at 0.3 0 RCIM_Cubic>]}
      }})
    R1 DefaultColor=<rgba 0 1 0.5972 0.7765>
    R2 DefaultColor:LinearColor=<rgba 1 0 0.8 1>
    Use R1 Painted Overrides=true
    Use R2 Painted Overrides=true
   ) {enabled=false}
   &VC_InitializeRibbon_v1(
    Custom Offset=true
    Enable Skeletal Position=true
    Facing Axis:ENiagaraOrientationAxis=::YAxis
    Lifetime=&MergeHelper_FloatSwitch(0=0.85, 1=0.8, 2=0.8, 3=?, Index=$System.GameModeID, Num=::NewEnumerator1)
    Local Vector Offset=&AddVectorToPosition(
     Position=&MergeHelper_VectorSwitch(0=<-12 0 0>, 1=<-40 0 0>, 2=<-40 0 0>, 3=?, Index=$System.GameModeID, Num=::NewEnumerator1)
     Vector=Linked() {linkedVariable {name=User.DTCarOffset002, type=NiagaraPosition}}
    )
    Offset Amount=1
    Ribbon Width=0
   )
   &AddVelocity(Velocity=&MergeHelper_VectorSwitch(0=<-1000 0 0>, 1=<-1000 0 0>, 2=<-1000 0 0>, 3=?, Index=$System.GameModeID, Num=::NewEnumerator1)) {module=AddVelocity001}
   /SetVariables(
    Particles.R1_Width:NiagaraFloat=&SelectFloatFromArray(Direct Array Index=&SpawnGroup(), Float Selection Array=DataInterface() {propertyValues {FloatData [1,0]}})
    Particles.R2_Width:NiagaraFloat=&SelectFloatFromArray(Direct Array Index=&SpawnGroup(), Float Selection Array=DataInterface() {propertyValues {FloatData [0,-1]}})
   ) {isSetParameters=true}
  ]
  particleUpdateScript [
   &ParticleState()
   &InheritSourceMovement(
    Applied Position Scale=&Multiply_VectorByFloat(
     Vector=<1 1 1>
     Float=&FloatFromCurve(FloatCurve=DataInterface() {propertyValues {Curve {keys [{interpMode=RCIM_Cubic, tangentMode=RCTM_Break, time=0.5, value=1}, {interpMode=RCIM_Cubic, tangentMode=RCTM_Break, time=1, value=0.5}]}}})
    )
    ApplyForce=false
    ApplyPosition=true
    ApplyPositionFalloff=false
   )
   &SolveForcesAndVelocity(Clamp Velocity=false, Limit Acceleration=false)
   /SetVariables(Particles.DotProduct:NiagaraFloat=&DotProduct(
     Vector A=&TransformVectorByMatrix(
      Vector:Vector3f=<1 0 0>
      Transformation Matrix:NiagaraMatrix=$Engine.Owner.SystemLocalToWorld
      Transform Vector as Position:NiagaraBool=false
     )
     Vector B=&NormalizeVector(
      Vector To Normalize=&SubtractVectorFromPosition(
       Position=$Engine.Owner.Position
       Vector=&ReturnCameraPosition(Coordinate Space:ENiagaraCoordinateSpace=::Simulation, Camera Query:NiagaraDataInterfaceCamera=DataInterface())
      )
      Safe Normalize=true
      Safe Fallback Threshold=0
      Safe Fallback Vector=<0 0 0.001>
     )
    )) {isSetParameters=true}
   /SetVariables(Particles.CustomPosition_R1:NiagaraPosition=&AddVectorToPosition(
     Position=$Particles.Position
     Vector=&MakeVector(
      X=0
      Y=0
      Z=&MaskFloat(
       Mask Source=0
       Mask Parameter=?
       Mask=0
       Masked Float=&Cosine(
        Normalized Angle=&Add_Float(A=0, B=&Subtract_Float(A=$Emitter.NormalizedLoopAge, B=$Particles.NormalizedAge))
        Period=1
        Scale=&FloatFromCurve(
         FloatCurve=DataInterface() {propertyValues {Curve {keys [
             {arriveTangent=4.1101, interpMode=RCIM_Cubic, leaveTangent=4.1101, tangentMode=RCTM_User, time=0.0999, value=0.71}
             {arriveTangent=-0.294, interpMode=RCIM_Cubic, leaveTangent=-0.294, time=0.5, value=1.6}
             {
              arriveTangent=-0.5881
              arriveTangentWeight=0.5973
              interpMode=RCIM_Cubic
              leaveTangent=-0.5881
              leaveTangentWeight=0.5973
              tangentMode=RCTM_User
              time=0.7585
              value=0.5164
             }
            ]}}}
         Scale Curve=&Lerp_Float(
          A=5
          B=6.25
          Alpha=&NormalizeFloat(Value=&Power(a=&Add_Float(A=$Particles.DotProduct, B=0.1), b=5), Min=0, Max=1)
         )
        )
        Bias=0
       )
       Passthrough Float=&Sine(
        Normalized Angle=&Add_Float(A=0, B=&Subtract_Float(A=$Emitter.NormalizedLoopAge, B=$Particles.NormalizedAge))
        Period=1
        Scale=&FloatFromCurve(
         FloatCurve=DataInterface() {propertyValues {Curve {keys [
             {arriveTangent=4.1101, interpMode=RCIM_Cubic, leaveTangent=4.1101, tangentMode=RCTM_User, time=0.0999, value=0.71}
             {arriveTangent=-0.294, interpMode=RCIM_Cubic, leaveTangent=-0.294, time=0.5, value=1.6}
             {
              arriveTangent=-0.5881
              arriveTangentWeight=0.5973
              interpMode=RCIM_Cubic
              leaveTangent=-0.5881
              leaveTangentWeight=0.5973
              tangentMode=RCTM_User
              time=0.7585
              value=0.5164
             }
            ]}}}
         Scale Curve=&Lerp_Float(
          A=5
          B=6.25
          Alpha=&NormalizeFloat(Value=&Power(a=&Add_Float(A=$Particles.DotProduct, B=0.1), b=5), Min=0, Max=1)
         )
        )
        Bias=0
       )
      )
     )
    )) {isSetParameters=true}
   &VC_UpdateRibbon_v1(Enable Skeletal Position=true)
   /SetVariables(
    Particles.R1_Width:NiagaraFloat=&Multiply_Float(
     A=$Particles.Initial.R1_Width
     B=&FloatFromCurve(
      FloatCurve=DataInterface() {propertyValues {Curve {keys [
          {interpMode=RCIM_Cubic, leaveTangentWeight=0.07, tangentMode=RCTM_Break, time=0, value=0.3}
          {
           arriveTangent=0.0116
           arriveTangentWeight=0.092
           interpMode=RCIM_Cubic
           leaveTangent=0.0116
           leaveTangentWeight=0.1363
           tangentMode=RCTM_User
           time=0.14
           value=1.2
          }
          {arriveTangent=-1.0302, interpMode=RCIM_Cubic, leaveTangent=-1.0302, tangentMode=RCTM_User, time=0.5, value=1}
          <at 0.7 0.85 RCIM_Cubic>
         ]}}}
      Scale Curve=60
     )
    )
    Particles.R2_Width:NiagaraFloat=&Multiply_Float(
     A=$Particles.Initial.R2_Width
     B=&FloatFromCurve(
      FloatCurve=DataInterface() {propertyValues {Curve {keys [
          {interpMode=RCIM_Cubic, leaveTangentWeight=0.07, tangentMode=RCTM_Break, time=0, value=0.3}
          {
           arriveTangent=0.0116
           arriveTangentWeight=0.092
           interpMode=RCIM_Cubic
           leaveTangent=0.0116
           leaveTangentWeight=0.1363
           tangentMode=RCTM_User
           time=0.14
           value=1.2
          }
          {arriveTangent=-1.0302, interpMode=RCIM_Cubic, leaveTangent=-1.0302, tangentMode=RCTM_User, time=0.5, value=1}
          <at 0.7 0.85 RCIM_Cubic>
         ]}}}
      Scale Curve=60
     )
    )
   ) {isSetParameters=true, module=SetVariables_#2}
   /SetVariables(
    Particles.R1Color:LinearColor=&MergeHelper_ColorSwitch(
     0=&ColorFromCurve(ColorCurve=DataInterface() {propertyValues {
        AlphaCurve {keys [{arriveTangent=7.1293, interpMode=RCIM_Cubic, time=0, value=3.5}, <at 0.5 3.5>, <at 0.75 0.5>, <at 0.89 0 RCIM_Cubic>]}
        BlueCurve {keys [<at 0 0.5391 RCIM_Cubic>,<at 0.9143 0.5043 RCIM_Cubic>]}
        GreenCurve {keys [<at 0 0.2131 RCIM_Cubic>,<at 0.9143 0.1096 RCIM_Cubic>]}
        RedCurve {keys [<at 0 0.0141 RCIM_Cubic>,<at 0.9143 0.3652 RCIM_Cubic>]}
       }})
     1=&ColorFromCurve(ColorCurve=DataInterface() {propertyValues {
        AlphaCurve {keys [
          <at 0.0971 0>
          {arriveTangent=8.6878, interpMode=RCIM_Cubic, leaveTangent=8.6878, time=0.2081, value=2}
          <at 0.5 3.5>
          <at 0.75 0.5>
          <at 0.89 0 RCIM_Cubic>
         ]}
        BlueCurve {keys [<at 0 0.5391 RCIM_Cubic>,<at 0.9143 0.5043 RCIM_Cubic>]}
        GreenCurve {keys [<at 0 0.2131 RCIM_Cubic>,<at 0.9143 0.1096 RCIM_Cubic>]}
        RedCurve {keys [<at 0 0.0141 RCIM_Cubic>,<at 0.9143 0.3652 RCIM_Cubic>]}
       }})
     2=&ColorFromCurve(ColorCurve=DataInterface() {propertyValues {
        AlphaCurve {keys [
          <at 0.0971 0>
          {arriveTangent=8.6878, interpMode=RCIM_Cubic, leaveTangent=8.6878, time=0.2081, value=2}
          <at 0.5 3.5>
          <at 0.75 0.5>
          <at 0.89 0 RCIM_Cubic>
         ]}
        BlueCurve {keys [<at 0 0.5391 RCIM_Cubic>,<at 0.9143 0.5043 RCIM_Cubic>]}
        GreenCurve {keys [<at 0 0.2131 RCIM_Cubic>,<at 0.9143 0.1096 RCIM_Cubic>]}
        RedCurve {keys [<at 0 0.0141 RCIM_Cubic>,<at 0.9143 0.3652 RCIM_Cubic>]}
       }})
     3=?
     Index=$System.GameModeID
     Num=::NewEnumerator1
    )
    Particles.R2Color:LinearColor=&ColorFromCurve(ColorCurve=DataInterface() {propertyValues {
       AlphaCurve {keys [<at 0 1.5 RCIM_Cubic>,<at 0.5 1.5>,<at 0.75 0.5>,<at 0.86 0 RCIM_Cubic>]}
       BlueCurve {keys [<at 0 0.7913 RCIM_Cubic>,<at 1 0.7913 RCIM_Cubic>]}
       GreenCurve {keys [<at 0 0.6856 RCIM_Cubic>,<at 1 0.6856 RCIM_Cubic>]}
       RedCurve {keys [<at 0 0.461 RCIM_Cubic>,<at 1 0.461 RCIM_Cubic>]}
      }})
   ) {isSetParameters=true}
  ]
  rendererClasses [&NiagaraRibbonRendererProperties]
  renderers [
   {rendererClass=&NiagaraRibbonRendererProperties}
   {rendererClass=&NiagaraRibbonRendererProperties, rendererIndex=1}
   {rendererClass=&NiagaraRibbonRendererProperties, rendererIndex=2}
   {rendererClass=&NiagaraRibbonRendererProperties, rendererIndex=3}
  ]
 }
 NE_VC_Boost_Sprite_Parent {
  emitterUpdateScript [
   &EmitterState()
   &VC_FilteredSocketCount()
   &SpawnRate(SpawnRate=8, Use Spawn Probability=false)
   &SpawnRate(SpawnGroup=1, SpawnRate=8, Use Spawn Probability=false) {module=SpawnRate001}
  ]
  particleSpawnScript [
   &VC_PaintedVariantsOverride(
    Black=<rgba 1 1 1 1>
    BurntSienna=<rgba 0.5 0.1058 0 1>
    Cobalt=<rgba 0.429 0.4488 1.5 1>
    Crimson=<rgba 1.5 0.3 0.3325 1>
    DefaultColor=<rgba 1.25 0.75 0.25 1>
    ForestGreen=<rgba 0.3464 0.5 0.3541 1>
    Grey=<rgba 0.9 0.9 0.9 0.8>
    Lime=<rgba 0.25 0.5 0 1>
    Orange=<rgba 1 0.335 0.05 1>
    Pink=<rgba 1 0.5 0.8182 1>
    Purple=<rgba 0.6952 0.3438 1.2 1>
    SkyBlue=<rgba 0.2 0.7323 1 1>
   )
   &InitializeParticle(
    Color=$Particles.Color
    Lifetime=&RandomRangeFloat(Minimum=1, Maximum=1.2)
    Sprite Rotation Angle=&RandomRangeFloat(Minimum=0, Maximum=1)
    Sprite Rotation Mode=::NewEnumerator1
    Sprite Size Max=<0.5 0.5>
    Sprite Size Min=<0.15 0.15>
    Sprite Size Mode=::NewEnumerator2
    UsePositionOffset=false
    Write Material Random=true
   )
   &PartitionParticles(Number Of Partitions=$Emitter.PortCount)
   &SkeletalMeshLocation(
    Bone / Socket Index=$Particles.PartitionParticles.Partition
    Bone / Socket Sampling Mode=::NewEnumerator5
    MaskSpawnGroup=false
    Mesh Sampling Type=::NewEnumerator6
    Sampled Position Offset=&AddVectorToPosition(Position=<-20 0 0>, Vector=$User.DTCarOffset)
    Skeletal Mesh=$System.SkeletalMesh
   )
   &AddVelocity(Velocity=&RandomRangeVector(Minimum=<-200 0 10>, Maximum=<-500 0 -10>), Velocity Speed Scale=0.8)
   &Color(Color=$Particles.Color)
   /SetVariables(Particles.FakeGlow:Vector2f=<15 15>) {isSetParameters=true}
   &ShapeLocation(MaskSpawnGroup=false, Override Seed=false, Sphere Radius=42, UseOwnerScale=true)
  ]
  particleUpdateScript [
   &ParticleState()
   &CurlNoiseForce(
    Bias Noise=false
    Noise Frequency=50000
    Noise Strength=&RandomRangeFloat(Minimum=100, Maximum=150)
    Pan Noise=false
    Randomization Vector:Vector3f=<0.65 0.125 0>
   )
   &InheritSourceMovement(
    Applied Position Scale=&VectorFromFloat(Value=&FloatFromCurve(FloatCurve=DataInterface() {propertyValues {Curve {keys [{interpMode=RCIM_Cubic, tangentMode=RCTM_Break, time=0.3, value=1}, {interpMode=RCIM_Cubic, tangentMode=RCTM_Break, time=1, value=0}]}}}))
    ApplyForce=false
    ApplyPosition=true
    ApplyPositionFalloff=false
   )
   &ScaleSpriteSize(Scale Sprite Size Mode=::NewEnumerator0, Uniform Scale Factor=12)
   &Drag(
    Drag=&RandomRangeFloat(
     Minimum=0
     Maximum=&FloatFromCurve(FloatCurve=DataInterface() {propertyValues {Curve {defaultValue=1, keys [<at 0.2 0 RCIM_Cubic>,<at 1 1 RCIM_Cubic>]}}}, Scale Curve=2)
    )
    Use Linear Drag=true
    Use Rotational Drag=true
   )
   &SolveForcesAndVelocity(Clamp Velocity=false, Limit Acceleration=false)
   &ScaleColor(
    Scale Alpha=1.5
    Scale RGB=&VectorFromFloat(Value=&MakeCustomFloatFromBool(Bool=$User.bIsPainted, True Float=15, False Float=15))
    ScaleA=true
    ScaleRGB=true
   )
   &JitterPosition(Jitter Amount:NiagaraFloat=&RandomRangeFloat(Minimum=-4, Maximum=4), Jitter Delay=2)
  ]
  rendererClasses [&NiagaraSpriteRendererProperties]
  renderers [
   {rendererClass=&NiagaraSpriteRendererProperties}
   {rendererClass=&NiagaraSpriteRendererProperties, rendererIndex=1}
  ]
 }
 NE_VC_Boost_Sprite_Parent001 {
  emitterUpdateScript [
   &EmitterState()
   &VC_FilteredSocketCount()
   &SpawnRate(SpawnRate=4, Use Spawn Probability=false)
   &SpawnRate(SpawnGroup=1, SpawnRate=6, Use Spawn Probability=false) {module=SpawnRate001}
  ]
  particleSpawnScript [
   &VC_PaintedVariantsOverride(
    Black=<rgba 1 1 1 1>
    BurntSienna=<rgba 0.5 0.1058 0 1>
    Cobalt=<rgba 0.429 0.4488 1.5 1>
    Crimson=<rgba 1.5 0.3 0.3325 1>
    DefaultColor=<rgba 0.2241 0.5211 0.6609 0.9217>
    ForestGreen=<rgba 0.3464 0.5 0.3541 1>
    Grey=<rgba 0.9 0.9 0.9 0.8>
    Lime=<rgba 0.25 0.5 0 1>
    Orange=<rgba 1 0.335 0.05 1>
    Pink=<rgba 1 0.5 0.8182 1>
    Purple=<rgba 0.6952 0.3438 1.2 1>
    R1 DefaultColor=<rgba 0.03 0.6896 1 1>
    SkyBlue=<rgba 0.2 0.7323 1 1>
    Use R1 Painted Overrides=true
    Use R2 Painted Overrides=true
   ) {enabled=false}
   /SetVariables(Particles.VisibilityTag=$Emitter.SpawnGroup) {isSetParameters=true}
   &InitializeParticle(
    Color Mode=::NewEnumerator0
    Lifetime=&RandomRangeFloat(Minimum=0.8, Maximum=1)
    Sprite Rotation Angle=&RandomRangeFloat(Minimum=0, Maximum=1)
    Sprite Rotation Mode=::NewEnumerator1
    Sprite Size Mode=::NewEnumerator3
    UsePositionOffset=false
    Write Material Random=true
   )
   &PartitionParticles(Number Of Partitions=$Emitter.PortCount)
   &SkeletalMeshLocation(
    Bone / Socket Index=$Particles.PartitionParticles.Partition
    Bone / Socket Sampling Mode=::NewEnumerator5
    MaskSpawnGroup=false
    Mesh Sampling Type=::NewEnumerator6
    Sampled Position Offset=&AddVectorToPosition(Position=<-20 0 0>, Vector=$User.DTCarOffset)
    Skeletal Mesh=$System.SkeletalMesh
   )
   &AddVelocity(Velocity=&RandomRangeVector(Minimum=<-250 0 -5>, Maximum=<-400 0 5>), Velocity Speed Scale=0.8)
   &ShapeLocation(
    MaskSpawnGroup=false
    Offset=<-100 0 0>
    Offset Coordinate Space=::Local
    Offset Mode=::NewEnumerator0
    Override Seed=false
    Sphere Radius=42
    UseOwnerScale=true
   )
  ]
  particleUpdateScript [
   &ParticleState()
   &CurlNoiseForce(
    Bias Noise=false
    Noise Frequency=50000
    Noise Strength=&RandomRangeFloat(Minimum=100, Maximum=150)
    Pan Noise=false
    Randomization Vector:Vector3f=<0.65 0.125 0>
   )
   &InheritSourceMovement(
    Applied Position Scale=&VectorFromFloat(Value=&FloatFromCurve(FloatCurve=DataInterface() {propertyValues {Curve {keys [{interpMode=RCIM_Cubic, tangentMode=RCTM_Break, time=0.3, value=1}, {interpMode=RCIM_Cubic, tangentMode=RCTM_Break, time=1, value=0}]}}}))
    ApplyForce=false
    ApplyPosition=true
    ApplyPositionFalloff=false
   )
   &ScaleSpriteSize(
    Scale Sprite Size Mode=::NewEnumerator0
    Uniform Scale Factor=&Multiply_Float(
     A=3
     B=&FloatFromCurve(FloatCurve=DataInterface() {propertyValues {Curve {defaultValue=1, keys [<at 0 0 RCIM_Cubic>,<at 0.3 1 RCIM_Cubic>]}}})
    )
   )
   &Drag(
    Drag=&RandomRangeFloat(
     Minimum=0
     Maximum=&FloatFromCurve(FloatCurve=DataInterface() {propertyValues {Curve {
         defaultValue=1
         keys [
          {arriveTangent=0.0905, interpMode=RCIM_Cubic, leaveTangent=0.0905, tangentMode=RCTM_User, time=0.2, value=0}
          {arriveTangent=1.0887, interpMode=RCIM_Cubic, leaveTangent=1.0887, tangentMode=RCTM_User, time=1, value=0.25}
         ]
        }}})
    )
    Use Linear Drag=true
    Use Rotational Drag=true
   )
   &SolveForcesAndVelocity(Clamp Velocity=false, Limit Acceleration=false)
   &DynamicMaterialParameters(
    Index 0 Param 2=4
    Index 0 Param 3=&FloatFromCurve(FloatCurve=DataInterface() {propertyValues {Curve {keys [<at 0 0>,<at 1 15>]}}})
    Module.Index0Float0Write=true
    Module.Index0Float1Write=true
    Module.Index0Float2Write=true
    Module.Index0Float3Write=true
   )
   /SetVariables(
    Particles.SP1_Color:LinearColor=&ColorFromCurve(ColorCurve=DataInterface() {propertyValues {
       AlphaCurve {keys [<at 0 1 RCIM_Cubic>,<at 0.5905 1>,<at 1 0 RCIM_Cubic>]}
       BlueCurve {keys [<at 0 0.4508 RCIM_Cubic>,<at 1 0.3565 RCIM_Cubic>]}
       GreenCurve {keys [<at 0 0.2122 RCIM_Cubic>,<at 1 0.0992 RCIM_Cubic>]}
       RedCurve {keys [<at 0 0.0666 RCIM_Cubic>,<at 1 0.2658 RCIM_Cubic>]}
      }})
    Particles.SP2_Color:LinearColor=&ColorFromCurve(ColorCurve=DataInterface() {propertyValues {
       AlphaCurve {defaultValue=1, keys [<at 0 1 RCIM_Cubic>,<at 0.5905 1>,<at 1 0 RCIM_Cubic>]}
       BlueCurve {keys [<at 0 0.6514 RCIM_Cubic>,<at 1 0.6514 RCIM_Cubic>]}
       GreenCurve {keys [<at 0 0.5647 RCIM_Cubic>,<at 1 0.5647 RCIM_Cubic>]}
       RedCurve {keys [<at 0 0.3763 RCIM_Cubic>,<at 1 0.3763 RCIM_Cubic>]}
      }})
   ) {isSetParameters=true}
   &JitterPosition(Jitter Amount:NiagaraFloat=&RandomRangeFloat(Minimum=-4, Maximum=4), Jitter Delay=2)
  ]
  rendererClasses [&NiagaraSpriteRendererProperties]
  renderers [
   {rendererClass=&NiagaraSpriteRendererProperties}
   {rendererClass=&NiagaraSpriteRendererProperties, rendererIndex=1}
   {rendererClass=&NiagaraSpriteRendererProperties, rendererIndex=2}
   {rendererClass=&NiagaraSpriteRendererProperties, rendererIndex=3}
  ]
 }
}
summary {
 systemName=NS_DecoBlue
 emitters [
  {emitterName=NE_VC_Boost_Sprite_Parent, rendererClasses [&NiagaraSpriteRendererProperties]}
  {emitterName=NE_VC_Boost_Ribbon_GoldBlue, rendererClasses [&NiagaraRibbonRendererProperties]}
  {emitterName=LensFlare, rendererClasses [&NiagaraSpriteRendererProperties]}
  {emitterName=NE_VC_Boost_Ribbon_Wind, rendererClasses [&NiagaraRibbonRendererProperties]}
  {emitterName=NE_VC_Boost_Sprite_Parent001, rendererClasses [&NiagaraSpriteRendererProperties]}
 ]
 userVariables [
  User.bIsFrontendPreview {type=NiagaraBool, defaultValue=true}
  User.bIsPainted {type=NiagaraBool}
  User.ColorVariantID {type=NiagaraInt32}
  User.DTCarOffset {type=Vector3f, description="", defaultValue {struct=Vector3f, value=<0 0 0>}}
  User.DTCarOffset001 {type=NiagaraPosition, description="", defaultValue {}}
  User.DTCarOffset002 {type=NiagaraPosition, description="", defaultValue {}}
  User.GameModeID {type=NiagaraInt32}
  User.PaintedColor {type=LinearColor, description="", defaultValue {struct=LinearColor, value=<rgba 1 1 1 1>}}
  User.PaintedScalar {type=NiagaraFloat, defaultValue=1, description=""}
  User.PortCount {type=NiagaraInt32, defaultValue=2}
  User.SingleBoostPos {type=Vector3f, description="", defaultValue {struct=Vector3f, value=<0 0 0>}}
 ]
}
